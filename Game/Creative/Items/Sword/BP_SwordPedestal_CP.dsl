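; Creative sword pedestal that spawns and displays an Infinity Blade actor at its spawn point.
; After spawning and on minigame state events, the authority creates and attaches a blade if none is valid, skipping preview actors.
; A destroy delegate tracks blade loss, and destroying the pedestal also destroys its blade.
; The construction script ties the Sword component's visibility to the editor, while blade events adjust its hidden-in-game flag.
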(event MinigameStateChangedDelegate(FortMinigameLogic)
  (params (Minigame :type Fort%20Minigame%20Object%20Reference) (NewMinigameState :type EFortMinigameState%20Enum :default "PreGame"))
  (switch Utilities|FlowControl|Switch|SwitchonEFortMinigameState NewMinigameState
    (:PreGame
      _)
    (:WaitingForCameras
      _))
  (CallFunction|ConditionallyCreateInfinityBlade))

(event EventPostFinishSpawning
  (|CallParentFunction)
  (CallFunction|ConditionallyCreateInfinityBlade))

(event Custom|ConditionallyCreateInfinityBlade
  (Networking|SwitchHasAuthority
    (:Authority
      (if (not (Building|IsPreviewBuildingActor))
        (Utilities|IsValid (Variables|Default|GetCurrentBlade)
          (:Is%20Not%20Valid
            (bind _returnvalue (Game|SpawnActorfromClass (Variables|Default|GetBladeClass) (Transformation|GetWorldTransform (Variables|Default|GetSpawnLocation)) "AlwaysSpawn" "SelectDefaultAtRuntime" self 0 0))
            (bind _output_get (Variables|Default|SetCurrentBlade _returnvalue))
            (Game|BindEventtoOnDestroyed _output_get (delegate Custom|OnCurrentBladeDestroyed))
            (Transformation|AttachActorToComponent _output_get (Variables|Default|GetSpawnLocation) "None" "SnapToTarget" "SnapToTarget" "SnapToTarget" true)
            (Networking|SetNetDormancy _output_get "DORM_Awake")
            (Development|SetHiddeninGame (Variables|Default|GetSword) false true)))))))

(event Custom|OnCurrentBladeDestroyed
  (params (DestroyedActor :type Actor%20Object%20Reference))
  (if (Utilities|Equal(Object) DestroyedActor (Variables|Default|GetCurrentBlade))
    (Variables|Default|SetCurrentBlade 0)
    (Development|SetHiddeninGame (Variables|Default|GetSword) true true)))

(event EventDestroyed
  (bind _currentblade (Variables|Default|GetCurrentBlade))
  (Utilities|IsValid _currentblade
    (:Is%20Valid
      (Game|UnbindEventfromOnDestroyed _currentblade (EventDispatchers|CreateEvent :!event Custom|OnCurrentBladeDestroyed))
      (Actor|DestroyActor _currentblade))))

(fn ConstructionScript
  (Rendering|SetVisibility (Variables|Default|GetSword) (Editor|InEditor) false))
